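# Control Rig for the Frogman medium male character body, which procedurally orients the chest knife and its loop bone on top of the skeleton.
# It rotates dyn_driver_knife so its X axis points at dyn_knife_aim, using a from-two-vectors quaternion applied to the aim bone's rotation.
# A weighted transform constraint then makes dyn_parent_loop follow mostly the knife driver, blended with clavicle_l.
# Debug axis draws are left enabled along the chain.
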ControlRigBlueprint /BRCosmetics/Characters/Player/Male/Medium/Bodies/M_MED_Frogman/Meshes/M_MED_Frogman_CtrlRig
graphs [
 RigVMFunctionLibrary {links [], nodes []}
 RigVMModel {nodes [
   n1 = &RigUnit_BeginExecution()
   /Script/ControlRig.RigUnit_DebugTransformMutable(
    Color=(R=1,G=0,B=0,A=1)
    DebugDrawSettings=(DepthPriority=SDPG_Foreground,Lifetime=-1)
    Mode=Axes
    Scale=3
    Space=dyn_knife_aim
    Thickness=0.5
    Transform=(Rotation=(X=0,Y=0,Z=0,W=1),Translation=(X=0,Y=0,Z=0),Scale3D=(X=1,Y=1,Z=1))
    WorldOffset=(Rotation=(X=0,Y=0,Z=0,W=1),Translation=(X=0,Y=0,Z=0),Scale3D=(X=1,Y=1,Z=1))
    bEnabled=True
    ExecutePin=@n1.ExecutePin
   ) {is=struct RigVMUnitNode}
   > &RigUnit_DebugTransformMutable(
    Color=(R=1,G=0,B=0,A=1)
    DebugDrawSettings=(DepthPriority=SDPG_Foreground,Lifetime=-1)
    Mode=Axes
    Scale=20
    Space=dyn_driver_knife
    Thickness=0.2
    Transform=(Rotation=(X=0,Y=0,Z=0,W=1),Translation=(X=0,Y=0,Z=0),Scale3D=(X=1,Y=1,Z=1))
    WorldOffset=(Rotation=(X=0,Y=0,Z=0,W=1),Translation=(X=0,Y=0,Z=0),Scale3D=(X=1,Y=1,Z=1))
    bEnabled=True
   )
   n2 = &RigUnit_GetBoneTransform(Bone=dyn_knife_aim)
   n3 = /Script/RigVM.RigVMFunction_MathQuaternionGetAxis(Axis=X, Quaternion=@n2.Transform.Rotation) {is=struct RigVMUnitNode, template="Axis::Execute(in Quaternion,in Axis,out Result)"}
   n4 = /Script/RigVM.RigVMFunction_MathVectorAdd(A=(X=0,Y=0,Z=0), B=(X=0,Y=0,Z=0)) {is=struct RigVMUnitNode, template="Add::Execute(in A,in B,out Result)"}
   n5 = /Script/RigVM.RigVMFunction_MathVectorSub(
    A=@n2.Transform.Translation
    B=/Script/RigVM.RigVMFunction_MathTransformTransformVector(
     Transform=&RigUnit_GetBoneTransform(Bone=dyn_driver_knife).Transform
     Location=/Script/RigVM.RigVMFunction_MathVectorSelectBool(Condition=true, IfFalse=(X=0,Y=0,Z=0), IfTrue=@n4.Result) {is=struct RigVMUnitNode}.Result
    ) {is=struct RigVMUnitNode}.Result
   ) {is=struct RigVMUnitNode, template="Subtract::Execute(in A,in B,out Result)"}
   > /Script/ControlRig.RigUnit_SetBoneRotation(
    Bone=dyn_driver_knife
    Space=GlobalSpace
    Weight=1
    bPropagateToChildren=True
    Rotation=/Script/RigVM.RigVMFunction_MathQuaternionMul(
     B=@n2.Transform.Rotation
     A=/Script/RigVM.RigVMFunction_MathQuaternionFromTwoVectors(A=@n3.Result, B=@n5.Result) {is=struct RigVMUnitNode, template="FromTwoVectors::Execute(in A,in B,out Result)"}.Result
    ) {is=struct RigVMUnitNode, template="Multiply::Execute(in A,in B,out Result)"}.Result
   ) {is=struct RigVMUnitNode}
   > &RigUnit_DebugTransformMutable(
    Color=(R=1,G=0,B=0,A=1)
    DebugDrawSettings=(DepthPriority=SDPG_Foreground,Lifetime=-1)
    Mode=Axes
    Scale=9.5
    Space=None
    Thickness=0.2
    WorldOffset=(Rotation=(X=0,Y=0,Z=0,W=1),Translation=(X=0,Y=0,Z=0),Scale3D=(X=1,Y=1,Z=1))
    bEnabled=True
    Transform=&RigUnit_GetBoneTransform(Bone=dyn_driver_knife).Transform
   )
   > /Script/ControlRig.RigUnit_TransformConstraint(
    BaseBone=None
    BaseTransform=(Rotation=(X=0,Y=0,Z=0,W=1),Translation=(X=0,Y=0,Z=0),Scale3D=(X=1,Y=1,Z=1))
    BaseTransformSpace=GlobalSpace
    Bone=dyn_parent_loop
    Targets="((Transform=(Rotation=(X=0,Y=0,Z=0,W=1),Translation=(X=0,Y=0,Z=0),Scale3D=(X=1,Y=1,Z=1)),Weight=0.91,bMaintainOffset=True,Filter=(TranslationFilter=(bX=True,bY=True,bZ=True),RotationFilter=(bX=True,bY=True,bZ=True),ScaleFilter=(bX=True,bY=True,bZ=True))),( ...(truncated)"
    bUseInitialTransforms=True
    Targets.0.Transform=&RigUnit_GetBoneTransform(Bone=dyn_driver_knife).Transform
    Targets.1.Transform=&RigUnit_GetBoneTransform(Bone=clavicle_l).Transform
   ) {is=struct RigVMUnitNode}
   > &RigUnit_DebugTransformMutable(
    Color=(R=1,G=0,B=0,A=1)
    DebugDrawSettings=(DepthPriority=SDPG_Foreground,Lifetime=-1)
    Mode=Axes
    Scale=5
    Space=None
    Thickness=0.25
    WorldOffset=(Rotation=(X=0,Y=0,Z=0,W=1),Translation=(X=0,Y=0,Z=0),Scale3D=(X=1,Y=1,Z=1))
    bEnabled=True
    Transform=&RigUnit_GetBoneTransform(Bone=dyn_parent_loop).Transform
   )
  ]}
]
hierarchy [
 root {loc=<0 0 0>, rot=<0 0 0 1>}
 pelvis {parent=root, loc=<0 -0.7071 93.1191>, rot=<0.0127 0.707 -0.0127 -0.707>}
 spine_01 {parent=pelvis, loc=<6.7494 -0.6388 0>, rot=<0 0 0.034 -0.9994>}
 spine_02 {parent=spine_01, loc=<7.2518 -0.035 0>, rot=<0 0 0.0238 -0.9997>}
 spine_03 {parent=spine_02, loc=<6.0104 0 0>, rot=<0 0 0.0149 -0.9999>}
 spine_04 {parent=spine_03, loc=<8.1024 0 0>, rot=<0 0 -0.0132 -0.9999>}
 spine_05 {parent=spine_04, loc=<10.0146 0 0>, rot=<0 0 -0.0015 -1>}
 clavicle_l {parent=spine_05, loc=<4.3575 -0.1466 -2.3223>, rot=<0.0465 -0.7289 -0.009 -0.6829>}
 upperarm_l {parent=clavicle_l, loc=<15.5111 0 0>, rot=<0.0226 -0.3234 -0.0088 -0.9459>}
 lowerarm_l {parent=upperarm_l, loc=<26.6514 0 0>, rot=<0.0391 -0.0099 -0.2453 -0.9686>}
 hand_l {parent=lowerarm_l, loc=<22.3491 0 0>, rot=<-0.7081 -0.0507 0.0975 -0.6975>}
 index_metacarpal_l {parent=hand_l, loc=<1.6535 0.0183 -2.96>, rot=<0.0158 -0.0124 -0.0173 -0.9996>}
 index_01_l {parent=index_metacarpal_l, loc=<8.3008 0.0357 -0.0524>, rot=<-0.0044 0.0485 0.0924 -0.9945>}
 index_02_l {parent=index_01_l, loc=<3.3587 -0.3004 0.0452>, rot=<-0.002 0.0027 0.3225 -0.9466>}
 index_03_l {parent=index_02_l, loc=<2.5578 -0.123 -0.0252>, rot=<-0.002 0.0067 0.2558 -0.9667>}
 middle_metacarpal_l {parent=hand_l, loc=<2.1856 -0.1725 -0.5984>, rot=<0.006 0.0168 -0.0329 -0.9993>}
 middle_01_l {parent=middle_metacarpal_l, loc=<7.3275 -0.2447 0.0129>, rot=<-0.0208 0.0042 0.0858 -0.9961>}
 middle_02_l {parent=middle_01_l, loc=<4.0123 0.1401 0.1175>, rot=<-0.0218 0.0258 0.3512 -0.9357>}
 middle_03_l {parent=middle_02_l, loc=<2.6077 0.0534 0.1373>, rot=<0.0082 0.0188 0.1708 -0.9851>}
 pinky_metacarpal_l {parent=hand_l, loc=<1.7616 -0.9753 3.3814>, rot=<-0.0397 0.0814 -0.069 -0.9935>}
 pinky_01_l {parent=pinky_metacarpal_l, loc=<7.1709 -0.4831 -0.1094>, rot=<-0.0654 -0.0035 0.2242 -0.9723>}
 pinky_02_l {parent=pinky_01_l, loc=<2.8274 0.0893 -0.1043>, rot=<0.0037 0.0011 0.2854 -0.9584>}
 pinky_03_l {parent=pinky_02_l, loc=<2.2046 0.0307 -0.0078>, rot=<-0.0002 0.0098 0.1445 -0.9895>}
 ring_metacarpal_l {parent=hand_l, loc=<2.2074 -0.3625 1.3638>, rot=<-0.0284 0.0722 -0.0458 -0.9959>}
 ring_01_l {parent=ring_metacarpal_l, loc=<7.0295 -0.6197 -0.1446>, rot=<-0.0489 -0.0252 0.1892 -0.9804>}
 ring_02_l {parent=ring_01_l, loc=<3.6174 0.4646 -0.1027>, rot=<0 0 0.2806 -0.9598>}
 ring_03_l {parent=ring_02_l, loc=<2.4152 0.1141 0.0199>, rot=<0 0 0.1189 -0.9929>}
 thumb_01_l {parent=hand_l, loc=<1.3118 -0.3191 -2.459>, rot=<0.6546 -0.4042 0.0664 -0.6354>}
 thumb_02_l {parent=thumb_01_l, loc=<5.9336 -0.0448 0.0401>, rot=<-0.0037 0.0213 0.298 -0.9543>}
 thumb_03_l {parent=thumb_02_l, loc=<2.2223 0.3682 -0.0907>, rot=<0.0118 -0.013 0.0066 -0.9998>}
 lowerarm_twist_01_l {parent=lowerarm_l, loc=<17.8549 0 0>, rot=<0 0 0 1>}
 lowerarm_twist_02_l {parent=lowerarm_l, loc=<9.6199 0 0>, rot=<0 0 0 1>}
 upperarm_twist_01_l {parent=upperarm_l, loc=<0.0041 0 0>, rot=<0 0 0 -1>}
 upperarm_twist_02_l {parent=upperarm_l, loc=<13.7941 0 0>, rot=<0 0 0 -1>}
 clavicle_r {parent=spine_05, loc=<4.3575 -0.1466 2.3223>, rot=<0.7289 0.0465 -0.6829 0.009>}
 upperarm_r {parent=clavicle_r, loc=<-15.5111 0 0>, rot=<0.0226 -0.3234 -0.0088 -0.9459>}
 lowerarm_r {parent=upperarm_r, loc=<-26.6514 0 0>, rot=<0.0391 -0.0099 -0.2453 -0.9686>}
 hand_r {parent=lowerarm_r, loc=<-22.3492 0 0>, rot=<-0.7081 -0.0507 0.0975 -0.6975>}
 index_metacarpal_r {parent=hand_r, loc=<-1.6534 -0.0181 2.9601>, rot=<0.0158 -0.0124 -0.0173 -0.9996>}
 index_01_r {parent=index_metacarpal_r, loc=<-8.3008 -0.0357 0.0524>, rot=<-0.0044 0.0485 0.0924 -0.9945>}
 index_02_r {parent=index_01_r, loc=<-3.3589 0.3003 -0.0453>, rot=<-0.002 0.0027 0.3225 -0.9466>}
 index_03_r {parent=index_02_r, loc=<-2.5578 0.1231 0.0251>, rot=<-0.002 0.0067 0.2558 -0.9667>}
 middle_metacarpal_r {parent=hand_r, loc=<-2.1856 0.1726 0.5984>, rot=<0.006 0.0168 -0.0329 -0.9993>}
 middle_01_r {parent=middle_metacarpal_r, loc=<-7.3273 0.2447 -0.0129>, rot=<-0.0208 0.0042 0.0858 -0.9961>}
 middle_02_r {parent=middle_01_r, loc=<-4.0142 -0.1411 -0.07>, rot=<-0.0217 0.0259 0.3512 -0.9357>}
 middle_03_r {parent=middle_02_r, loc=<-2.6074 -0.0534 -0.1312>, rot=<0.0083 0.0188 0.1708 -0.9851>}
 pinky_metacarpal_r {parent=hand_r, loc=<-1.7617 0.9753 -3.3814>, rot=<-0.0397 0.0814 -0.069 -0.9935>}
 pinky_01_r {parent=pinky_metacarpal_r, loc=<-7.1707 0.4832 0.1094>, rot=<-0.0654 -0.0035 0.2242 -0.9723>}
 pinky_02_r {parent=pinky_01_r, loc=<-2.8274 -0.0892 0.1043>, rot=<0.0037 0.0011 0.2854 -0.9584>}
 pinky_03_r {parent=pinky_02_r, loc=<-2.2045 -0.031 0.0078>, rot=<-0.0002 0.0098 0.1445 -0.9895>}
 ring_metacarpal_r {parent=hand_r, loc=<-2.2073 0.3627 -1.3638>, rot=<-0.0284 0.0722 -0.0458 -0.9959>}
 ring_01_r {parent=ring_metacarpal_r, loc=<-7.0297 0.6197 0.1446>, rot=<-0.0489 -0.0252 0.1892 -0.9804>}
 ring_02_r {parent=ring_01_r, loc=<-3.6172 -0.4648 0.1027>, rot=<0 0 0.2806 -0.9598>}
 ring_03_r {parent=ring_02_r, loc=<-2.4152 -0.1141 -0.0199>, rot=<0 0 0.1189 -0.9929>}
 thumb_01_r {parent=hand_r, loc=<-1.3119 0.3192 2.459>, rot=<0.6546 -0.4042 0.0664 -0.6354>}
 thumb_02_r {parent=thumb_01_r, loc=<-5.9335 0.0447 -0.0401>, rot=<-0.0037 0.0213 0.298 -0.9543>}
 thumb_03_r {parent=thumb_02_r, loc=<-2.2223 -0.3682 0.0907>, rot=<0.0118 -0.013 0.0066 -0.9998>}
 lowerarm_twist_01_r {parent=lowerarm_r, loc=<-17.855 0 0>, rot=<0 0 0 -1>}
 lowerarm_twist_02_r {parent=lowerarm_r, loc=<-9.6199 0 0>, rot=<0 0 0 -1>}
 upperarm_twist_01_r {parent=upperarm_r, loc=<-0.0041 0 0>, rot=<0 0 0 1>}
 upperarm_twist_02_r {parent=upperarm_r, loc=<-13.7942 0 0>, rot=<0 0 0 1>}
 neck_01 {parent=spine_05, loc=<11.3138 0.1529 0>, rot=<0 0 -0.2003 -0.9797>}
 neck_02 {parent=neck_01, loc=<4.5627 0 0>, rot=<0 0 0.0504 -0.9987>}
 head {parent=neck_02, loc=<5.3914 0 0>, rot=<0 0 0.115 -0.9934>}
 dyn_parent_collar_bck {parent=spine_05, loc=<12.2541 -6.0329 0.0002>, rot=<0.0284 -0.7065 -0.0284 -0.7065>}
 dyn_driver_collar_bck {parent=dyn_parent_collar_bck, loc=<0 0 0>, rot=<-0.1211 0.6967 0.1211 -0.6967>}
 dyn_collar_bck_01 {parent=dyn_driver_collar_bck, loc=<0 0 0>, rot=<0 0 0 -1>}
 dyn_knife_aim {parent=spine_05, loc=<-0.4324 12.6892 -11.4144>, rot=<0.2048 0.2178 0.9377 0.177>}
 dyn_parent_knife {parent=spine_05, loc=<8.8717 8.0975 -14.7603>, rot=<0.2048 0.2178 0.9377 0.177>}
 dyn_driver_knife {parent=dyn_parent_knife, loc=<0 0 0>, rot=<0 0 0 1>}
 dyn_parent_loop {parent=dyn_driver_knife, loc=<-14.229 -1.228 0.8772>, rot=<-0.5067 -0.0036 0.8084 -0.2997>}
 dyn_driver_loop {parent=dyn_parent_loop, loc=<0 0 0>, rot=<0.1678 -0.7928 0.4298 0.3982>}
 dyn_collar_fr_l_MSTR {parent=spine_05, loc=<3.3314 7.9873 -8.3182>, rot=<0.9235 0.3447 -0.1532 -0.07>}
 dyn_collar_fr_l_01 {parent=dyn_collar_fr_l_MSTR, loc=<0.8292 0 0>, rot=<0 -0.0561 0 -0.9984>}
 dyn_collar_fr_l_02 {parent=dyn_collar_fr_l_01, loc=<5.663 -0.4924 0.7959>, rot=<-0.0479 0.1389 0.5158 -0.844>}
 dyn_collar_fr_l_03 {parent=dyn_collar_fr_l_02, loc=<2.6506 0 0>, rot=<0 0 0 1>}
 dyn_collar_fr_r_MSTR {parent=spine_05, loc=<0.9818 9.7869 7.3328>, rot=<0.8048 0.4894 0.2488 0.2255>}
 dyn_collar_fr_r_01 {parent=dyn_collar_fr_r_MSTR, loc=<0.8292 0 0>, rot=<0 0 0 1>}
 dyn_collar_fr_r_02 {parent=dyn_collar_fr_r_01, loc=<5.5147 -0.1547 0.7264>, rot=<0 0 0 1>}
 dyn_parent_nadeA {parent=spine_02, loc=<-4.9319 6.1433 14.412>, rot=<0.0282 -0.7065 -0.0282 -0.7065>}
 dyn_driver_nadeA {parent=dyn_parent_nadeA, loc=<0 0 0>, rot=<0 0 0 1>}
 dyn_nadeA_01 {parent=dyn_driver_nadeA, loc=<0 0 0>, rot=<-0.4269 -0.5361 0.6685 -0.289>}
 dyn_nadeA_02 {parent=dyn_nadeA_01, loc=<2.1617 0 0>, rot=<0 0 0 1>}
 dyn_parent_nadeB {parent=spine_02, loc=<-4.273 0.5209 15.1112>, rot=<0.0282 -0.7065 -0.0282 -0.7065>}
 dyn_driver_nadeB {parent=dyn_parent_nadeB, loc=<0 0 0>, rot=<-0.3004 -0.6034 0.6728 -0.3051>}
 dyn_nadeB_01 {parent=dyn_driver_nadeB, loc=<0 0 0>, rot=<0 0 0 -1>}
 dyn_nadeB_02 {parent=dyn_nadeB_01, loc=<2.3272 0 0>, rot=<0 0 0 -1>}
 dyn_main_side_pouch {parent=spine_02, loc=<-0.1301 3.3759 11.6741>, rot=<0.0282 -0.7065 -0.0282 -0.7065>}
 dyn_parent_side_pouch {parent=dyn_main_side_pouch, loc=<-2.1603 0 6.2422>, rot=<0 0 0 1>}
 dyn_side_pouch_01 {parent=dyn_parent_side_pouch, loc=<0 0 0>, rot=<-0.127 -0.7066 0.6956 -0.0255>}
 dyn_side_pouch_02 {parent=dyn_side_pouch_01, loc=<2.7544 0 0>, rot=<-0.0369 -0.0151 -0.5354 -0.8436>}
 dyn_side_pouch_03 {parent=dyn_side_pouch_02, loc=<3.94 0 0>, rot=<0 0 0 1>}
 dyn_parent_front_packs_02 {parent=spine_02, loc=<2.0814 9.3897 -7.6292>, rot=<0.0282 -0.7065 -0.0282 -0.7065>}
 dyn_parent_front_packs_01 {parent=spine_01, loc=<7.8944 9.1845 -7.6292>, rot=<0.0114 -0.707 -0.0114 -0.707>}
 thigh_l {parent=pelvis, loc=<-2.5693 -0.0109 -7.7501>, rot=<0.0012 0.0572 0.0155 -0.9982>}
 calf_l {parent=thigh_l, loc=<-39.0484 0 0>, rot=<0.0006 -0.0161 -0.0391 -0.9991>}
 calf_twist_01_l {parent=calf_l, loc=<-35.8131 -0.0967 0.0397>, rot=<0.0237 0.0095 -0.0035 -0.9997>}
 calf_twist_02_l {parent=calf_l, loc=<-18.8064 -0.0967 0.0397>, rot=<0.0237 0.0095 -0.0035 -0.9997>}
 foot_l {parent=calf_l, loc=<-42.561 0.1077 0.4292>, rot=<0.0418 -0.0455 0.0343 -0.9975>}
 ball_l {parent=foot_l, loc=<-6.1599 13.8264 0.6806>, rot=<0.0001 -0.0001 -0.7063 -0.7079>}
 dyn_laceA_l {parent=foot_l, loc=<5.108 3.9505 2.4784>, rot=<-0.0357 -0.7045 -0.0284 -0.7082>}
 dyn_laceB_l {parent=foot_l, loc=<5.0578 3.4724 2.9319>, rot=<-0.0357 -0.7045 -0.0284 -0.7082>}
 thigh_twist_01_l {parent=thigh_l, loc=<-21.6303 0 0>, rot=<0.0013 0 0 -1>}
 dyn_parent_pouch_l {parent=thigh_l, loc=<-12.4304 5.3406 -7.4904>, rot=<-0.0019 -0.7464 0.0016 -0.6655>}
 dyn_driver_pouch_l {parent=dyn_parent_pouch_l, loc=<0 0 0>, rot=<0.6122 -0.3466 -0.5712 -0.4228>}
 dyn_pouch_01_l {parent=dyn_driver_pouch_l, loc=<0 0 0>, rot=<0 0 0 1>}
 dyn_pouch_tab {parent=dyn_pouch_01_l, loc=<1.2317 -4.1342 0.6991>, rot=<0.1259 -0.0021 0.217 -0.968>}
 thigh_r {parent=pelvis, loc=<-2.5691 -0.0109 7.7501>, rot=<0.0572 -0.0012 0.9982 0.0155>}
 calf_r {parent=thigh_r, loc=<39.0487 0 0>, rot=<0.0006 -0.0161 -0.0391 -0.9991>}
 calf_twist_01_r {parent=calf_r, loc=<35.8132 0.0967 -0.0397>, rot=<0.0237 0.0095 -0.0035 -0.9997>}
 calf_twist_02_r {parent=calf_r, loc=<18.8064 0.0967 -0.0397>, rot=<0.0237 0.0095 -0.0035 -0.9997>}
 foot_r {parent=calf_r, loc=<42.5611 -0.1077 -0.4292>, rot=<0.0418 -0.0455 0.0343 -0.9975>}
 ball_r {parent=foot_r, loc=<6.1598 -13.8264 -0.6806>, rot=<0.0001 -0.0001 -0.7063 -0.7079>}
 dyn_laceA_r {parent=foot_r, loc=<-5.3908 -4.4915 -2.8037>, rot=<0.7082 0.0284 -0.7045 -0.0357>}
 dyn_laceB_r {parent=foot_r, loc=<-5.0497 -4.6769 -2.2718>, rot=<0.7082 0.0284 -0.7045 -0.0357>}
 thigh_twist_01_r {parent=thigh_r, loc=<21.6303 0 0>, rot=<0.0013 0 0 -1>}
 dyn_parent_pouch_r {parent=thigh_r, loc=<23.6232 -3.8778 5.9919>, rot=<-0.6655 0.0016 0.7464 0.0019>}
 dyn_pouch_r {parent=dyn_parent_pouch_r, loc=<0 0 0>, rot=<-0.5398 -0.4098 0.5794 -0.4528>}
 dyn_jacket_r_fr_MSTR {parent=pelvis, loc=<11.5308 5.0077 10.2035>, rot=<0.102 0.5148 0.8497 -0.0509>}
 dyn_jacket_r_fr_01 {parent=dyn_jacket_r_fr_MSTR, loc=<0 0 0>, rot=<0 0 0 -1>}
 dyn_jacket_r_fr_02 {parent=dyn_jacket_r_fr_01, loc=<8.5542 0 0>, rot=<0 0 0 -1>}
 dyn_followA_aim {parent=dyn_jacket_r_fr_02, loc=<14.1786 2.62 1.4454>, rot=<0.9309 0.0612 0.018 -0.3597>}
 dyn_jacket_r_bk_MSTR {parent=pelvis, loc=<3.9254 -11.6116 8.3088>, rot=<0.0842 -0.604 0.792 -0.0269>}
 dyn_jacket_r_bk_01 {parent=dyn_jacket_r_bk_MSTR, loc=<0 0 0>, rot=<0 0 0 1>}
 dyn_jacket_r_bk_02 {parent=dyn_jacket_r_bk_01, loc=<6.7952 0 0>, rot=<0 0 0 1>}
 dyn_jacket_l_bk_MSTR {parent=pelvis, loc=<3.9253 -11.6117 -8.3088>, rot=<0.0269 -0.792 0.604 -0.0842>}
 dyn_jacket_l_bk_01 {parent=dyn_jacket_l_bk_MSTR, loc=<0 0 0>, rot=<0 0 0 1>}
 dyn_jacket_l_bk_02 {parent=dyn_jacket_l_bk_01, loc=<6.7951 0 0>, rot=<0 0 0 1>}
 dyn_jacket_l_fr_MSTR {parent=pelvis, loc=<11.4416 5.2235 -9.8022>, rot=<0.0509 -0.8497 -0.5148 -0.102>}
 dyn_jacket_l_fr_01 {parent=dyn_jacket_l_fr_MSTR, loc=<0 0 0>, rot=<0 0 0 1>}
 dyn_jacket_l_fr_02 {parent=dyn_jacket_l_fr_01, loc=<8.554 0 0>, rot=<0 0 0 1>}
 dyn_jacket_r_MSTR {parent=pelvis, loc=<7.789 -1.1103 13.3281>, rot=<0.0961 -0.0017 0.9952 0.0178>}
 dyn_jacket_r_01 {parent=dyn_jacket_r_MSTR, loc=<0 0 0>, rot=<0 0 0 1>}
 dyn_jacket_r_02 {parent=dyn_jacket_r_01, loc=<8.4968 0 -0.0001>, rot=<0 0 0 1>}
 dyn_followB_aim {parent=dyn_jacket_r_02, loc=<9.346 -0.0726 2.0106>, rot=<0.6992 0.0516 0.0233 -0.7127>}
 dyn_jacket_l_MSTR {parent=pelvis, loc=<7.9631 -1.1165 -14.2245>, rot=<-0.0178 -0.9952 0.0017 -0.0961>}
 dyn_jacket_l_01 {parent=dyn_jacket_l_MSTR, loc=<0 0 0>, rot=<0 0 0 1>}
 dyn_jacket_l_02 {parent=dyn_jacket_l_01, loc=<8.4964 0 0>, rot=<0 0 0 1>}
 dyn_jacket_bk_MSTR {parent=pelvis, loc=<3.0734 -11.8996 0>, rot=<0.0619 -0.7044 0.7044 -0.0619>}
 dyn_jacket_bk {parent=dyn_jacket_bk_MSTR, loc=<0 0 0>, rot=<0 0 0 1>}
 dyn_jacket_bk_l_01 {parent=dyn_jacket_bk, loc=<6.7487 0.8351 -0.0269>, rot=<0 0 -0.0513 -0.9987>}
 dyn_jacket_bk_r_01 {parent=dyn_jacket_bk, loc=<6.7488 -0.8351 -0.0269>, rot=<0 0 0.0513 -0.9987>}
 dyn_parent_followA {parent=pelvis, loc=<9.7352 4.6498 10.9894>, rot=<-0.0127 -0.707 0.0127 -0.707>}
 dyn_driver_followA {parent=dyn_parent_followA, loc=<0 0 0>, rot=<-0.4555 -0.5006 0.6499 -0.3457>}
 dyn_parent_followB {parent=pelvis, loc=<9.5567 -0.3302 12.9006>, rot=<-0.0127 -0.707 0.0127 -0.707>}
 dyn_driver_followB {parent=dyn_parent_followB, loc=<0 0 0>, rot=<-0.4133 -0.5544 0.5829 -0.4266>}
 dyn_jacket_l_01 {type=Curve, loc=<0 0 0>, rot=<0 0 0 1>}
 dyn_jacket_r_01 {type=Curve, loc=<0 0 0>, rot=<0 0 0 1>}
 dyn_jacket_r_fr_01 {type=Curve, loc=<0 0 0>, rot=<0 0 0 1>}
 dyn_driver_nadeA {type=Curve, loc=<0 0 0>, rot=<0 0 0 1>}
 dyn_driver_nadeB {type=Curve, loc=<0 0 0>, rot=<0 0 0 1>}
]
