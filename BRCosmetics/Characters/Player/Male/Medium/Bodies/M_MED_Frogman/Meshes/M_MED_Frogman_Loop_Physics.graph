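# Physics asset for the loop part of the Frogman medium male player body cosmetic, covering its dyn_driver_loop and dyn_parent_loop bones.
# Each bone has a small capsule body, with the driver's capsule longer and offset along the bone.
# A single constraint joins driver to parent, with a wide first swing limit and near-zero second swing and twist limits, giving hinge-like limits.
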PhysicsAsset /BRCosmetics/Characters/Player/Male/Medium/Bodies/M_MED_Frogman/Meshes/M_MED_Frogman_Loop_Physics
bodies {
 dyn_driver_loop [{shapeType=Capsule, length=2, radius=1, shapeName=dyn_driver_loop_capsule, center=<2 0 0>, rotation=<rot 90 0 0>}]
 dyn_parent_loop [{shapeType=Capsule, length=0.505, radius=0.505, shapeName=dyn_parent_loop_capsule, rotation=<rot 0 -90 0>}]
}
constraints [{bone1Name=dyn_driver_loop, bone2Name=dyn_parent_loop, swing1LimitDegrees=100, swing2LimitDegrees=1, twistLimitDegrees=1}]
