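# Physics asset for the medium male Frogman player body cosmetic, defining bodies for dangling gear: two grenades, a side pouch and a right-side jacket piece.
# Each forms a short constrained chain hanging from a driver, parent or master bone body.
# Tight twist limits keep pieces from spinning, with pouch and grenades swinging mostly on one axis and the jacket equally on both.
# An unconstrained right-thigh capsule is the only main-skeleton body, apparently a collider for the gear.
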PhysicsAsset /BRCosmetics/Characters/Player/Male/Medium/Bodies/M_MED_Frogman/Meshes/M_MED_Frogman_Physics
bodies {
 dyn_driver_nadeA [{shapeType=Capsule, length=0.505, radius=0.505, rotation=<rot 0 -90 0>}]
 dyn_driver_nadeB [{shapeType=Capsule, length=0.505, radius=0.505, rotation=<rot 0 -90 0>}]
 dyn_jacket_r_01 [{shapeType=Box, extentX=12.5, extentY=43, extentZ=2, center=<5 0 -0.1304>}]
 dyn_jacket_r_02 [{shapeType=Box, extentX=12, extentY=48, extentZ=2, center=<5 0 -0.2915>}]
 dyn_jacket_r_MSTR [{shapeType=Box, extentX=1.01, extentY=1.01, extentZ=1.01, rotation=<rot 0 -90 0>}]
 dyn_nadeA_01 [{shapeType=Box, extentX=3.5, extentY=1, extentZ=3.5, center=<0.35 0 0>}]
 dyn_nadeA_02 [
  {shapeType=Capsule, length=4.5183, radius=2.0935, center=<5.7487 0.0535 0.0084>, rotation=<rot 0 100.2856 -89.4922>}
  {shapeType=Capsule, length=1.25, radius=3.5, center=<11.3932 0.0001 0.2208>, rotation=<rot -88.8897 -0.0031 0.0031>}
 ]
 dyn_nadeB_01 [{shapeType=Box, extentX=3.5, extentY=1, extentZ=3.5, center=<0.25 0 0>}]
 dyn_nadeB_02 [
  {shapeType=Sphere, radius=4, center=<15.6975 3.6123 -0.2251>}
  {shapeType=Capsule, length=6.778, radius=2.5631, center=<5.3564 1.6405 -0.1626>, rotation=<rot 0 -65.4615 -89.0914>}
 ]
 dyn_parent_side_pouch [{shapeType=Box, extentX=14.51, extentY=4.01, extentZ=7.51, center=<-0.3722 -0.2656 -5.6866>, rotation=<rot 0.6407 -102.4967 8.1152>}]
 dyn_side_pouch_01 [{shapeType=Box, extentX=5, extentY=3, extentZ=15.5, center=<0.2 0 0>}]
 dyn_side_pouch_02 [{shapeType=Box, extentX=8, extentY=2.5, extentZ=18, center=<0.6254 0 0>}]
 dyn_side_pouch_03 [{shapeType=Box, extentX=3, extentY=1, extentZ=1.5, center=<0.3871 0 0>}]
 thigh_r [{shapeType=Capsule, length=16.5, radius=6.5, center=<10.8867 0.033 -0.9999>, rotation=<rot 90 0 0>}]
}
constraints [
 {bone1Name=dyn_side_pouch_01, bone2Name=dyn_parent_side_pouch, swing1LimitDegrees=25, swing2LimitDegrees=1, twistLimitDegrees=1}
 {bone1Name=dyn_side_pouch_02, bone2Name=dyn_side_pouch_01, swing1LimitDegrees=20, swing2LimitDegrees=3.4286, twistLimitDegrees=1}
 {bone1Name=dyn_side_pouch_03, bone2Name=dyn_side_pouch_02, swing1LimitDegrees=21, swing2LimitDegrees=7.8571, twistLimitDegrees=1}
 {bone1Name=dyn_nadeB_01, bone2Name=dyn_driver_nadeB, swing1LimitDegrees=45, swing2LimitDegrees=10, twistLimitDegrees=1}
 {bone1Name=dyn_nadeA_01, bone2Name=dyn_driver_nadeA, swing1LimitDegrees=15, swing2LimitDegrees=10, twistLimitDegrees=1}
 {bone1Name=dyn_nadeB_02, bone2Name=dyn_nadeB_01, swing1LimitDegrees=30, swing2LimitDegrees=22.8571, twistLimitDegrees=1}
 {bone1Name=dyn_nadeA_02, bone2Name=dyn_nadeA_01, swing1LimitDegrees=25, swing2LimitDegrees=16, twistLimitDegrees=1}
 {bone1Name=dyn_jacket_r_01, bone2Name=dyn_jacket_r_MSTR, swing1LimitDegrees=45, swing2LimitDegrees=45, twistLimitDegrees=5}
 {bone1Name=dyn_jacket_r_02, bone2Name=dyn_jacket_r_01, swing1LimitDegrees=45, swing2LimitDegrees=45, twistLimitDegrees=5}
]
