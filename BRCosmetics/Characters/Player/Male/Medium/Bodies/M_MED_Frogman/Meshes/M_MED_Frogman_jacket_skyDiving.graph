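# Physics asset for the Frogman player character's jacket in its skydiving variant, setting up dynamic jacket bones to move against the body.
# Box-shaped panels hang in short limited chains from tiny MSTR capsules, which attach to spine_01 with free angular motion.
# Cross-constraints tie neighboring panel tips together so the jacket moves as a connected skirt, though the left-front master has no panels.
# Capsules on pelvis, lower spine and thigh twists stand in for the body.
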PhysicsAsset /BRCosmetics/Characters/Player/Male/Medium/Bodies/M_MED_Frogman/Meshes/M_MED_Frogman_jacket_skyDiving
bodies {
 dyn_jacket_bk [{shapeType=Box, extentX=12.5, extentY=19.5, extentZ=3, center=<3 0 0>}]
 dyn_jacket_bk_MSTR [{shapeType=Capsule, length=1, radius=1, rotation=<rot 90 0 0>}]
 dyn_jacket_bk_l_01 [{shapeType=Box, extentX=11.5, extentY=10.5, extentZ=3, center=<3 0 0>}]
 dyn_jacket_bk_r_01 [{shapeType=Box, extentX=12.5, extentY=9, extentZ=3, center=<3 0 0>}]
 dyn_jacket_l_01 [{shapeType=Box, extentX=10, extentY=18, extentZ=2, center=<5 0 0>}]
 dyn_jacket_l_02 [{shapeType=Box, extentX=10, extentY=18, extentZ=2, center=<5 0 0>}]
 dyn_jacket_l_MSTR [{shapeType=Capsule, length=1, radius=1, rotation=<rot 90 0 0>}]
 dyn_jacket_l_bk_01 [{shapeType=Box, extentX=12.5, extentY=14.5, extentZ=4.5, center=<3 0 0>}]
 dyn_jacket_l_bk_02 [{shapeType=Box, extentX=12, extentY=14.5, extentZ=5, center=<4 0 0>}]
 dyn_jacket_l_bk_MSTR [{shapeType=Capsule, length=1, radius=1, rotation=<rot 90 0 0>}]
 dyn_jacket_l_fr_MSTR [{shapeType=Capsule, length=1, radius=1, rotation=<rot 90 0 0>}]
 dyn_jacket_r_01 [{shapeType=Box, extentX=11.5, extentY=17, extentZ=3, center=<4.1573 0 0.367>}]
 dyn_jacket_r_02 [{shapeType=Box, extentX=12.5, extentY=17.5, extentZ=3.5, center=<5 0 0.1336>, rotation=<rot 3.0847 0 0>}]
 dyn_jacket_r_MSTR [{shapeType=Capsule, length=1, radius=1, rotation=<rot 90 0 0>}]
 dyn_jacket_r_bk_01 [{shapeType=Box, extentX=13.5, extentY=16.5, extentZ=4.5, center=<3 0 0>}]
 dyn_jacket_r_bk_02 [{shapeType=Box, extentX=11.5, extentY=17, extentZ=4.5, center=<4 0 0>}]
 dyn_jacket_r_bk_MSTR [{shapeType=Capsule, length=1, radius=1, rotation=<rot 90 0 0>}]
 dyn_jacket_r_fr_01 [{shapeType=Box, extentX=10, extentY=11, extentZ=2.5, center=<5 -0.502 0.6621>, rotation=<rot 0 0 5.8016>}]
 dyn_jacket_r_fr_02 [{shapeType=Box, extentX=16, extentY=11, extentZ=2.5, center=<7.0156 -0.4306 1.3507>, rotation=<rot 5.4049 0.278 2.9487>}]
 dyn_jacket_r_fr_MSTR [{shapeType=Capsule, length=1, radius=1, rotation=<rot 90 0 0>}]
 pelvis [
  {shapeType=Capsule, length=9.248, radius=10.5529, center=<-4.1837 -1.4406 -0.0001>}
  {shapeType=Capsule, length=13.0936, radius=8.1762, center=<-7.049 -1.2079 0>}
 ]
 spine_01 [{shapeType=Capsule, length=10.2698, radius=7.3782, center=<-1.215 -1.0321 0>}]
 thigh_twist_01_l [{shapeType=Capsule, length=16, radius=7, center=<5 -1 1.421>, rotation=<rot 90 0 0>}]
 thigh_twist_01_r [{shapeType=Capsule, length=15.75, radius=7.25, center=<-6.1884 0.033 -0.9999>, rotation=<rot 90 0 0>}]
}
constraints [
 {bone1Name=spine_01, bone2Name=pelvis, swing1LimitDegrees=45, swing2LimitDegrees=45, twistLimitDegrees=45}
 {bone1Name=dyn_jacket_bk, bone2Name=dyn_jacket_bk_MSTR, swing1LimitDegrees=5, swing2LimitDegrees=45, twistLimitDegrees=5}
 {bone1Name=dyn_jacket_bk_l_01, bone2Name=dyn_jacket_bk, swing1LimitDegrees=30, swing2LimitDegrees=30, twistLimitDegrees=5}
 {bone1Name=dyn_jacket_bk_r_01, bone2Name=dyn_jacket_bk, swing1LimitDegrees=30, swing2LimitDegrees=30, twistLimitDegrees=5}
 {bone1Name=dyn_jacket_r_bk_01, bone2Name=dyn_jacket_r_bk_MSTR, swing1LimitDegrees=45, swing2LimitDegrees=45, twistLimitDegrees=5}
 {bone1Name=dyn_jacket_r_bk_02, bone2Name=dyn_jacket_r_bk_01, swing1LimitDegrees=45, swing2LimitDegrees=45, twistLimitDegrees=5}
 {bone1Name=dyn_jacket_l_bk_01, bone2Name=dyn_jacket_l_bk_MSTR, swing1LimitDegrees=45, swing2LimitDegrees=45, twistLimitDegrees=5}
 {bone1Name=dyn_jacket_l_bk_02, bone2Name=dyn_jacket_l_bk_01, swing1LimitDegrees=45, swing2LimitDegrees=45, twistLimitDegrees=5}
 {bone1Name=dyn_jacket_r_fr_01, bone2Name=dyn_jacket_r_fr_MSTR, swing1LimitDegrees=5, swing2LimitDegrees=45, twistLimitDegrees=5}
 {bone1Name=dyn_jacket_r_fr_02, bone2Name=dyn_jacket_r_fr_01, swing1LimitDegrees=5, swing2LimitDegrees=45, twistLimitDegrees=5}
 {bone1Name=dyn_jacket_l_02, bone2Name=dyn_jacket_l_01, swing1LimitDegrees=45, swing2LimitDegrees=45, twistLimitDegrees=5}
 {bone1Name=dyn_jacket_l_01, bone2Name=dyn_jacket_l_MSTR, swing1LimitDegrees=45, swing2LimitDegrees=45, twistLimitDegrees=5}
 {bone1Name=dyn_jacket_r_01, bone2Name=dyn_jacket_r_MSTR, swing1LimitDegrees=45, swing2LimitDegrees=45, twistLimitDegrees=5}
 {bone1Name=dyn_jacket_r_02, bone2Name=dyn_jacket_r_01, swing1LimitDegrees=45, swing2LimitDegrees=45, twistLimitDegrees=5}
 {
  bone1Name=thigh_twist_01_l
  bone2Name=pelvis
  swing1LimitDegrees=45
  swing1Motion=Free
  swing2LimitDegrees=45
  swing2Motion=Free
  twistLimitDegrees=45
  twistMotion=Free
 }
 {
  bone1Name=thigh_twist_01_r
  bone2Name=pelvis
  swing1LimitDegrees=45
  swing1Motion=Free
  swing2LimitDegrees=45
  swing2Motion=Free
  twistLimitDegrees=45
  twistMotion=Free
 }
 {
  bone1Name=dyn_jacket_bk_MSTR
  bone2Name=spine_01
  swing1LimitDegrees=45
  swing1Motion=Free
  swing2LimitDegrees=45
  swing2Motion=Free
  twistLimitDegrees=45
  twistMotion=Free
 }
 {
  bone1Name=dyn_jacket_r_bk_MSTR
  bone2Name=spine_01
  swing1LimitDegrees=45
  swing1Motion=Free
  swing2LimitDegrees=45
  swing2Motion=Free
  twistLimitDegrees=45
  twistMotion=Free
 }
 {
  bone1Name=dyn_jacket_l_bk_MSTR
  bone2Name=spine_01
  swing1LimitDegrees=45
  swing1Motion=Free
  swing2LimitDegrees=45
  swing2Motion=Free
  twistLimitDegrees=45
  twistMotion=Free
 }
 {
  bone1Name=dyn_jacket_l_fr_MSTR
  bone2Name=spine_01
  swing1LimitDegrees=45
  swing1Motion=Free
  swing2LimitDegrees=45
  swing2Motion=Free
  twistLimitDegrees=45
  twistMotion=Free
 }
 {
  bone1Name=dyn_jacket_r_fr_MSTR
  bone2Name=spine_01
  swing1LimitDegrees=45
  swing1Motion=Free
  swing2LimitDegrees=45
  swing2Motion=Free
  twistLimitDegrees=45
  twistMotion=Free
 }
 {
  bone1Name=dyn_jacket_l_MSTR
  bone2Name=spine_01
  swing1LimitDegrees=45
  swing1Motion=Free
  swing2LimitDegrees=45
  swing2Motion=Free
  twistLimitDegrees=45
  twistMotion=Free
 }
 {
  bone1Name=dyn_jacket_r_MSTR
  bone2Name=spine_01
  swing1LimitDegrees=45
  swing1Motion=Free
  swing2LimitDegrees=45
  swing2Motion=Free
  twistLimitDegrees=45
  twistMotion=Free
 }
 {bone1Name=dyn_jacket_r_02, bone2Name=dyn_jacket_r_fr_02, swing1LimitDegrees=45, swing2LimitDegrees=30, twistLimitDegrees=5}
 {bone1Name=dyn_jacket_r_bk_02, bone2Name=dyn_jacket_bk_r_01, swing1LimitDegrees=45, swing2LimitDegrees=30, twistLimitDegrees=5}
 {bone1Name=dyn_jacket_l_bk_02, bone2Name=dyn_jacket_bk_l_01, swing1LimitDegrees=45, swing2LimitDegrees=30, twistLimitDegrees=5}
 {bone1Name=dyn_jacket_l_02, bone2Name=dyn_jacket_l_bk_02, swing1LimitDegrees=45, swing2LimitDegrees=30, twistLimitDegrees=5}
 {bone1Name=dyn_jacket_r_02, bone2Name=dyn_jacket_r_bk_02, swing1LimitDegrees=45, swing2LimitDegrees=30, twistLimitDegrees=5}
]
