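# Post-process AnimBP for the Frogman player body cosmetic, adding secondary motion to jacket, pouches, knife and laces over the parent mesh's copied pose.
# Pose-driven bone offsets, constraints, AnimDynamics laces and a Control Rig build a cached pose.
# RigidBody sims layer on top, switching jacket physics for skydiving, while thigh-driven jacket offsets replace them when rigid bodies are disabled.
# The event graph updates gravity override through a deprecated helper.
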AnimBlueprint /BRCosmetics/Characters/Player/Male/Medium/Bodies/M_MED_Frogman/Meshes/M_MED_Frogman_AnimBP
target_skeleton=&/M_MED_Frogman_Skeleton
graphs [
 AnimGraph {nodes [
   Anim Dynamics - Bone: dyn_driver_loop(Alpha=1, ExternalForce="0,0,0") {
    angularDampingOverride=0.9
    bOverrideAngularDamping=true
    bUseGravityOverride=true
    simulationSpace=BoneRelative
    alphaBoolBlend {blendOption=Linear}
    boundBone {boneName=dyn_driver_loop}
    componentAppliedLinearAccClamp=<3000 3000 3000>
    componentLinearAccScale=<0.3 0.3 0.3>
    componentLinearVelScale=<0.3 0.3 0.3>
    gravityOverride {z=-980}
    physicsBodyDefinitions [{
      boundBone {boneName=dyn_driver_loop}
      boxExtents=<10 10 10>
      constraintSetup {angularLimitsMax=<0 0 60>, angularLimitsMin=<0 0 -35>, angularTarget=<0 0 0>}
      localJointOffset=<20 0 0>
     }]
    relativeSpaceBone {boneName=spine_05}
   }
   Apply Additive(Additive="Sequence Player (additive) 'M_MED_Frogman_Jacket_Pose'"() {sequence=&/M_MED_Frogman_Jacket_Pose}) {alphaBoolBlend {blendOption=Linear}}
   "Constraint - dyn_parent_loop (P)"(Alpha=1, ConstraintWeights_0=1) {
    alphaBoolBlend {blendOption=Linear}
    boneToModify {boneName=dyn_parent_loop}
    constraintSetup [{targetBone {boneName=dyn_driver_knife}}]
    constraintWeights [1]
   }
   "Transform (Modify) Bone - Bone: dyn_jacket_r_MSTR"(Alpha=1, Translation="1.884,-0.233,-1.514", Rotation="2,0.000,0.000", Scale="1,1,1") {
    rotationMode=BMM_Additive
    rotationSpace=BCS_BoneSpace
    translationMode=BMM_Additive
    translationSpace=BCS_BoneSpace
    alphaBoolBlend {blendOption=Linear}
    alphaScaleBiasClamp {bInterpResult=true, interpSpeedDecreasing=15, interpSpeedIncreasing=15}
    boneToModify {boneName=dyn_jacket_r_MSTR}
    rotation {pitch=2}
    translation=<1.884 -0.233 -1.514>
   }
   n1 = Break Fort Anim Input Skydiving External Force(FortAnimInput_SkydivingExternalForce=Get SkydivingExternalForceInput()) -> (
    bUseSkydivingVectorForce=~
    SkydivingVectorForce=~
    HeadToPelvisDirection=~
    FloatingMultiplier=~
    FloatingAdditive=~
    DivingMultiplier=~
    DivingAdditive=~
    ParachutingMultiplier=~
    ParachutingAdditive=~
   )
   n2 = RigidBody!(
    ComponentPose=Local To Component(LocalPose=Use cached pose 'Cache_PreSkyDive'())
    Alpha
    OverrideWorldGravity:Vector=Get GravityOverride()
    ExternalForce=@n1.SkydivingVectorForce
   ) -> (Pose) {
    bClampLinearTranslationLimitToRefPose=true
    bOverrideWorldGravity=true
    overridePhysicsAsset=&/M_MED_Frogman_jacket_dyn
    simulationSpace=BaseBoneSpace
    alphaBoolBlend {blendOption=Linear}
    baseBoneRef {boneName=spine_01}
    componentAppliedLinearAccClamp=<3000 3000 3000>
    componentLinearAccScale=<0.4 0.3 0.3>
    componentLinearVelScale=<0.4 0.3 0.3>
    overrideWorldGravity {z=-980}
   }
   n3 = Component To Local(ComponentPose=@n2)
   n4 = RigidBody!(ComponentPose=Local To Component(LocalPose=Use cached pose 'Cache_PreSkyDive'()), Alpha, ExternalForce=@n1.SkydivingVectorForce) -> (Pose) {
    bClampLinearTranslationLimitToRefPose=true
    bOverrideWorldGravity=true
    overridePhysicsAsset=&/M_MED_Frogman_jacket_skyDiving
    simulationSpace=BaseBoneSpace
    alphaBoolBlend {blendOption=Linear}
    baseBoneRef {boneName=spine_01}
    componentAppliedLinearAccClamp=<1100 1100 1100>
    componentLinearAccScale=<0.4 0.3 0.3>
    componentLinearVelScale=<0.4 0.3 0.3>
    overrideWorldGravity {z=-800}
   }
   n5 = Component To Local(ComponentPose=@n4)
   n6 = Get bIsSkyDiving()
   n7 = RigidBody(ComponentPose=Local To Component(LocalPose=Blend Poses by bool(BlendPose_0=@n5, BlendPose_1=@n3, bActiveValue=@n6) {blendType=Linear})) {
    bClampLinearTranslationLimitToRefPose=true
    bOverrideWorldGravity=true
    overridePhysicsAsset=&/M_MED_Frogman_Physics
    simulationSpace=BaseBoneSpace
    alphaBoolBlend {blendOption=Linear}
    baseBoneRef {boneName=root}
    componentAppliedLinearAccClamp=<1200 1200 1200>
    componentLinearAccScale=<0.8 0.8 0.35>
    componentLinearVelScale=<0.8 0.8 0.35>
    overrideWorldGravity {z=-980}
   }
   n8 = RigidBody!(ComponentPose=@n7, Alpha, OverrideWorldGravity:Vector=Get Gravity Override(), ExternalForce) -> (Pose) {
    bOverrideWorldGravity=true
    overridePhysicsAsset=&/M_MED_Frogman_Loop_Physics
    simulationSpace=BaseBoneSpace
    alphaBoolBlend {blendOption=Linear}
    baseBoneRef {boneName=clavicle_l}
    componentAppliedLinearAccClamp=<3000 3000 3000>
    componentLinearAccScale=<0.4 0.4 0.4>
    componentLinearVelScale=<0.4 0.4 0.4>
    overrideWorldGravity {z=-980}
    simSpaceSettings {
     maxAngularAcceleration=1200
     maxAngularVelocity=5
     maxLinearAcceleration=700
     maxLinearVelocity=700
     velocityScaleZ=0.1
     worldAlpha=0.6
     externalLinearDragV=<0.25 0.25 0.25>
    }
   }
   n9 = Get ThighUpL() -> (ThighUpL:float)
   n10 = "Transform (Modify) Bone - Bone: dyn_jacket_l_fr_01"(
    ComponentPose=Local To Component(LocalPose=Use cached pose 'AnimCache'())
    Alpha=@n9
    Translation="-2.7,0,0"
    Rotation="0,0,0"
    Scale="1,1,1"
   ) {
    rotationMode=BMM_Additive
    rotationSpace=BCS_BoneSpace
    translationMode=BMM_Additive
    translationSpace=BCS_BoneSpace
    alphaBoolBlend {blendOption=Linear}
    boneToModify {boneName=dyn_jacket_l_fr_01}
    translation {x=-2.7}
   }
   n11 = "Transform (Modify) Bone - Bone: dyn_jacket_l_fr_02"(ComponentPose=@n10, Alpha=@n9, Translation="-3.2,0,0", Rotation="95,-10,30.000", Scale="1,1,1") {
    rotationMode=BMM_Additive
    rotationSpace=BCS_BoneSpace
    translationMode=BMM_Additive
    translationSpace=BCS_BoneSpace
    alphaBoolBlend {blendOption=Linear}
    boneToModify {boneName=dyn_jacket_l_fr_02}
    rotation=<rot 95 -10 30>
    translation {x=-3.2}
   }
   n12 = "Transform (Modify) Bone - Bone: dyn_jacket_l_01"(ComponentPose=@n11, Alpha=@n9, Translation="-2.5,0,0", Rotation="15,15,-15", Scale="1,1,1") {
    rotationMode=BMM_Additive
    rotationSpace=BCS_BoneSpace
    translationMode=BMM_Additive
    translationSpace=BCS_BoneSpace
    alphaBoolBlend {blendOption=Linear}
    boneToModify {boneName=dyn_jacket_l_01}
    rotation=<rot 15 15 -15>
    translation {x=-2.5}
   }
   n13 = Get ThighUpR() -> (ThighUpR:float)
   n14 = "Transform (Modify) Bone - Bone: dyn_jacket_r_fr_01"(ComponentPose=@n12, Alpha=@n13, Translation="-2.7,0,0", Rotation="0,0,0", Scale="1,1,1") {
    rotationMode=BMM_Additive
    rotationSpace=BCS_BoneSpace
    translationMode=BMM_Additive
    translationSpace=BCS_BoneSpace
    alphaBoolBlend {blendOption=Linear}
    boneToModify {boneName=dyn_jacket_r_fr_01}
    translation {x=-2.7}
   }
   n15 = "Transform (Modify) Bone - Bone: dyn_jacket_r_fr_02"(ComponentPose=@n14, Alpha=@n13, Translation="-3.2,0,0", Rotation="95,10,-30", Scale="1,1,1") {
    rotationMode=BMM_Additive
    rotationSpace=BCS_BoneSpace
    translationMode=BMM_Additive
    translationSpace=BCS_BoneSpace
    alphaBoolBlend {blendOption=Linear}
    boneToModify {boneName=dyn_jacket_r_fr_02}
    rotation=<rot 95 10 -30>
    translation {x=-3.2}
   }
   n16 = "Transform (Modify) Bone - Bone: dyn_jacket_r_01"(ComponentPose=@n15, Alpha=@n13, Translation="-2.5,0,0", Rotation="15,-15,15", Scale="1,1,1") {
    rotationMode=BMM_Additive
    rotationSpace=BCS_BoneSpace
    translationMode=BMM_Additive
    translationSpace=BCS_BoneSpace
    alphaBoolBlend {blendOption=Linear}
    boneToModify {boneName=dyn_jacket_r_01}
    rotation=<rot 15 -15 15>
    translation {x=-2.5}
   }
   Output Pose(Result=Blend Poses by bool(BlendPose_0=Component To Local(ComponentPose=@n8), BlendPose_1=Component To Local(ComponentPose=@n16), bActiveValue=Get bRigidBodyEnabled()) {blendType=Linear}) {layerGroup=None}
   n17 = "Transform (Modify) Bone - Bone: dyn_collar_fr_r_MSTR"(
    ComponentPose=Local To Component(LocalPose=Copy Pose From Mesh() {bUseAttachedParent=true})
    Alpha=Get ARFiringPosition() -> (ARFiringPosition:float)
    Translation="-2.930,-0.000,0.154"
    Rotation="12.502,-12.061,-2.795"
    Scale="0.85,0.85,0.85"
   ) {
    rotationMode=BMM_Additive
    rotationSpace=BCS_BoneSpace
    scaleMode=BMM_Additive
    scaleSpace=BCS_BoneSpace
    translationMode=BMM_Additive
    translationSpace=BCS_BoneSpace
    alphaBoolBlend {blendOption=Linear}
    alphaScaleBiasClamp {bInterpResult=true, interpSpeedIncreasing=16}
    boneToModify {boneName=dyn_collar_fr_r_MSTR}
    rotation=<rot 12.502 -12.0613 -2.795>
    scale=<0.85 0.85 0.85>
    translation {x=-2.93, z=0.154}
   }
   n18 = "Constraint - dyn_parent_pouch_r (PP)"!(ComponentPose=@n17, Alpha=1, ConstraintWeights_0=0.3328, ConstraintWeights_1:float=0.6672) -> (Pose) {
    alphaBoolBlend {blendOption=Linear}
    boneToModify {boneName=dyn_parent_pouch_r}
    constraintSetup [{targetBone {boneName=thigh_r}}, {targetBone {boneName=thigh_twist_01_r}}]
    constraintWeights [1,1]
   }
   n19 = "Transform (Modify) Bone - Bone: dyn_pouch_r"(ComponentPose=@n18, Alpha=Get ThighUpR(), Translation="8,0,0", Rotation="0,0,0", Scale="1,1,1") {
    rotationMode=BMM_Additive
    rotationSpace=BCS_BoneSpace
    translationMode=BMM_Additive
    translationSpace=BCS_BoneSpace
    alphaBoolBlend {blendOption=Linear}
    boneToModify {boneName=dyn_pouch_r}
    translation {x=8}
   }
   n20 = "Constraint - dyn_followA_aim (PPP)"!(ComponentPose=@n19, Alpha=1, ConstraintWeights_0=0.3607, ConstraintWeights_1:float=0.6143, ConstraintWeights_2:float=0.0251) -> (Pose) {
    alphaBoolBlend {blendOption=Linear}
    boneToModify {boneName=dyn_followA_aim}
    constraintSetup [{targetBone {boneName=dyn_jacket_r_02}}, {targetBone {boneName=dyn_jacket_r_fr_02}}, {targetBone {boneName=dyn_jacket_r_bk_02}}]
    constraintWeights [1,1,1]
   }
   n21 = "Constraint - dyn_followB_aim (PPPPP)"!(
    ComponentPose=@n20
    Alpha=1
    ConstraintWeights_0=0.0106
    ConstraintWeights_1:float=0.0184
    ConstraintWeights_2:float=0.2555
    ConstraintWeights_3:float=0.1137
    ConstraintWeights_4:float=0.6019
   ) -> (Pose) {
    alphaBoolBlend {blendOption=Linear}
    boneToModify {boneName=dyn_followB_aim}
    constraintSetup [
     {targetBone {boneName=pelvis}}
     {targetBone {boneName=dyn_jacket_r_01}}
     {targetBone {boneName=dyn_jacket_r_fr_02}}
     {targetBone {boneName=dyn_jacket_r_bk_02}}
     {targetBone {boneName=dyn_jacket_r_02}}
    ]
    constraintWeights [1,1,1,1,1]
   }
   n22 = "Constraint - dyn_parent_nadeB (P)"!(ComponentPose=@n21, Alpha=1, ConstraintWeights_0=1, ConstraintWeights_1:float=0) -> (Pose) {
    alphaBoolBlend {blendOption=Linear}
    boneToModify {boneName=dyn_parent_nadeB}
    constraintSetup [{targetBone {boneName=dyn_main_side_pouch}}]
    constraintWeights [1,0]
   }
   n23 = "Constraint - dyn_parent_nadeA (P)"!(ComponentPose=@n22, Alpha=1, ConstraintWeights_0=1, ConstraintWeights_1:float=0) -> (Pose) {
    alphaBoolBlend {blendOption=Linear}
    boneToModify {boneName=dyn_parent_nadeA}
    constraintSetup [{targetBone {boneName=dyn_main_side_pouch}}]
    constraintWeights [1,0]
   }
   n24 = "Constraint - dyn_parent_knife (PPPPPPP)"!(
    ComponentPose=@n23
    Alpha=0.65
    ConstraintWeights_0=0.4591
    ConstraintWeights_1:float=0.0359
    ConstraintWeights_2:float=0.0268
    ConstraintWeights_3:float=0.0723
    ConstraintWeights_4:float=0.1158
    ConstraintWeights_5:float=0.1826
    ConstraintWeights_6:float=0.1075
   ) -> (Pose) {
    alphaBoolBlend {blendOption=Linear}
    boneToModify {boneName=dyn_parent_knife}
    constraintSetup [
     {targetBone {boneName=clavicle_l}}
     {targetBone {boneName=upperarm_twist_02_l}}
     {targetBone {boneName=spine_02}}
     {targetBone {boneName=spine_03}}
     {targetBone {boneName=spine_04}}
     {targetBone {boneName=spine_05}}
     {targetBone {boneName=upperarm_twist_01_l}}
    ]
    constraintWeights [1,1,1,1,1,1,1]
   }
   n25 = "Constraint - dyn_knife_aim (PPPPPPP)"!(
    ComponentPose=@n24
    Alpha=1
    ConstraintWeights_0=0.0131
    ConstraintWeights_1:float=0.0222
    ConstraintWeights_2:float=0.1248
    ConstraintWeights_3:float=0.2484
    ConstraintWeights_4:float=0.4096
    ConstraintWeights_5:float=0.1709
    ConstraintWeights_6:float=0.011
   ) -> (Pose) {
    alphaBoolBlend {blendOption=Linear}
    boneToModify {boneName=dyn_knife_aim}
    constraintSetup [
     {targetBone {boneName=clavicle_l}}
     {targetBone {boneName=spine_01}}
     {targetBone {boneName=spine_02}}
     {targetBone {boneName=spine_03}}
     {targetBone {boneName=spine_04}}
     {targetBone {boneName=spine_05}}
     {targetBone {boneName=upperarm_twist_01_l}}
    ]
    constraintWeights [1,1,1,1,1,1,1]
   }
   n26 = Anim Dynamics - Bone: dyn_laceA_l(ComponentPose=@n25, Alpha=0.55, ExternalForce="0,0,0") {
    angularDampingOverride=0.9999
    bGravityOverrideInSimSpace=true
    bLinearSpring=true
    bOverrideAngularDamping=true
    bOverrideLinearDamping=true
    gravityScale=0.33
    linearDampingOverride=0.999
    linearSpringConstant=150
    simulationSpace=RootRelative
    alphaBoolBlend {blendOption=Linear}
    boundBone {boneName=dyn_laceA_l}
    physicsBodyDefinitions [{
      sphereCollisionRadius=0
      boundBone {boneName=dyn_laceA_l}
      boxExtents=<10 10 10>
      constraintSetup {angularLimitsMax=<0 15 2>, angularLimitsMin=<0 -15 -2>, angularTarget=<0 0 0>, linearAxesMax=<4 2 0.02>, linearAxesMin=<-4 -2 -0.02>}
      localJointOffset=<0 0 -5>
     }]
   }
   n27 = Anim Dynamics - Bone: dyn_laceB_l(ComponentPose=@n26, Alpha=0.3, ExternalForce="0,0,0") {
    angularDampingOverride=0.9999
    bGravityOverrideInSimSpace=true
    bLinearSpring=true
    bOverrideAngularDamping=true
    bOverrideLinearDamping=true
    gravityScale=0.29
    linearDampingOverride=0.9999
    linearSpringConstant=150
    simulationSpace=RootRelative
    alphaBoolBlend {blendOption=Linear}
    boundBone {boneName=dyn_laceB_l}
    physicsBodyDefinitions [{
      sphereCollisionRadius=0
      boundBone {boneName=dyn_laceB_l}
      boxExtents=<10 10 10>
      constraintSetup {angularLimitsMax=<0 6 0>, angularLimitsMin=<0 -6 0>, angularTarget=<0 0 0>, linearAxesMax=<4 2 0.02>, linearAxesMin=<-5 -2 -0.02>}
      localJointOffset=<0 0 -6>
     }]
   }
   n28 = Anim Dynamics - Bone: dyn_laceB_r(ComponentPose=@n27, Alpha=0.55, ExternalForce="0,0,0") {
    bGravityOverrideInSimSpace=true
    bLinearSpring=true
    bOverrideLinearDamping=true
    gravityScale=0.33
    linearDampingOverride=0.995
    linearSpringConstant=150
    simulationSpace=RootRelative
    alphaBoolBlend {blendOption=Linear}
    boundBone {boneName=dyn_laceB_r}
    physicsBodyDefinitions [{
      sphereCollisionRadius=0
      boundBone {boneName=dyn_laceB_r}
      boxExtents=<10 10 10>
      constraintSetup {angularLimitsMax=<0 8 0>, angularLimitsMin=<0 -8 0>, angularTarget=<0 0 0>, linearAxesMax=<4 2 0.02>, linearAxesMin=<-4 -2 -0.02>}
      localJointOffset=<0 0 -5>
     }]
   }
   n29 = Anim Dynamics - Bone: dyn_laceA_r(ComponentPose=@n28, Alpha=0.3, ExternalForce="0,0,0") {
    bGravityOverrideInSimSpace=true
    bLinearSpring=true
    bOverrideLinearDamping=true
    gravityScale=0.29
    linearDampingOverride=0.995
    linearSpringConstant=150
    simulationSpace=RootRelative
    alphaBoolBlend {blendOption=Linear}
    boundBone {boneName=dyn_laceA_r}
    physicsBodyDefinitions [{
      sphereCollisionRadius=0
      boundBone {boneName=dyn_laceA_r}
      boxExtents=<10 10 10>
      constraintSetup {angularLimitsMax=<0 5 0>, angularLimitsMin=<0 -5 0>, angularTarget=<0 0 0>, linearAxesMax=<4 2 0.02>, linearAxesMin=<-5 -2 -0.02>}
      localJointOffset=<0 0 -6>
     }]
   }
   n30 = "Constraint - dyn_parent_pouch_l (P)"(ComponentPose=@n29, Alpha=1, ConstraintWeights_0=1) {
    alphaBoolBlend {blendOption=Linear}
    boneToModify {boneName=dyn_parent_pouch_l}
    constraintSetup [{targetBone {boneName=thigh_twist_01_l}}]
    constraintWeights [1]
   }
   n31 = "Transform (Modify) Bone - Bone: dyn_driver_pouch_l"(
    ComponentPose=@n30
    Alpha=Get CrouchAlpha() -> (CrouchAlpha:float)
    Translation="-0.000,-0.000,-3.102"
    Rotation="-0.080,0.010,-7.322"
    Scale="1,1,1"
   ) {
    rotationMode=BMM_Additive
    rotationSpace=BCS_BoneSpace
    translationMode=BMM_Additive
    translationSpace=BCS_BoneSpace
    alphaBoolBlend {blendOption=Linear}
    alphaScaleBiasClamp {bInterpResult=true, interpSpeedDecreasing=18, interpSpeedIncreasing=18}
    boneToModify {boneName=dyn_driver_pouch_l}
    rotation=<rot -0.08 0.01 -7.3222>
    translation {z=-3.102}
   }
   n32 = Get CrouchAlpha()
   n33 = "Transform (Modify) Bone - Bone: dyn_parent_front_packs_01"(ComponentPose=@n31, Alpha=@n32, Translation="0.562,1.415,3.000", Rotation="0.000,9.744,0.000", Scale="1,1,1") {
    rotationMode=BMM_Additive
    rotationSpace=BCS_BoneSpace
    translationMode=BMM_Additive
    translationSpace=BCS_BoneSpace
    alphaBoolBlend {blendOption=Linear}
    alphaScaleBiasClamp {bInterpResult=true, interpSpeedDecreasing=18, interpSpeedIncreasing=18}
    boneToModify {boneName=dyn_parent_front_packs_01}
    rotation {yaw=9.7442}
    translation=<0.562 1.415 3>
   }
   n34 = "Transform (Modify) Bone - Bone: dyn_parent_front_packs_02"(ComponentPose=@n33, Alpha=@n32, Translation="0.449,0.918,3.000", Rotation="0.000,10.599,0.000", Scale="1,1,1") {
    rotationMode=BMM_Additive
    rotationSpace=BCS_BoneSpace
    translationMode=BMM_Additive
    translationSpace=BCS_BoneSpace
    alphaBoolBlend {blendOption=Linear}
    alphaScaleBiasClamp {bInterpResult=true, interpSpeedDecreasing=18, interpSpeedIncreasing=18}
    boneToModify {boneName=dyn_parent_front_packs_02}
    rotation {yaw=10.5994}
    translation=<0.4491 0.9178 3>
   }
   n35 = Control Rig(Source=Component To Local(ComponentPose=@n34)) {
    alphaBoolBlend {blendOption=Linear}
    controlRigAssetReference {blueprintRigClass=&/M_MED_Frogman_CtrlRig.M_MED_Frogman_CtrlRig_C}
   }
   Save cached pose 'AnimCache'(Pose=@n35) {cacheName=AnimCache}
   n36 = "Transform (Modify) Bone - Bone: dyn_jacket_l_MSTR"(
    ComponentPose=Local To Component(LocalPose=Use cached pose 'AnimCache'())
    Alpha=Get ThighUpL()
    Translation="-2.5,-2,2"
    Rotation="14.252,-7.529,-19.671"
    Scale="1,1,1"
   ) {
    rotationMode=BMM_Additive
    rotationSpace=BCS_BoneSpace
    translationMode=BMM_Additive
    translationSpace=BCS_BoneSpace
    alphaBoolBlend {blendOption=Linear}
    alphaScaleBiasClamp {bInterpResult=true, interpSpeedDecreasing=15, interpSpeedIncreasing=15}
    boneToModify {boneName=dyn_jacket_l_MSTR}
    rotation=<rot 14.252 -7.5291 -19.671>
    translation=<-2.5 -2 2>
   }
   n37 = "Transform (Modify) Bone - Bone: dyn_jacket_r_fr_MSTR"(ComponentPose=@n36, Alpha=Get ThighUpR(), Translation="-0.909,3.976,10", Rotation="7.287,10.970,2.483", Scale="1,1,1") {
    rotationMode=BMM_Additive
    rotationSpace=BCS_BoneSpace
    translationMode=BMM_Additive
    translationSpace=BCS_BoneSpace
    alphaBoolBlend {blendOption=Linear}
    alphaScaleBiasClamp {bInterpResult=true, interpSpeedDecreasing=15, interpSpeedIncreasing=15}
    boneToModify {boneName=dyn_jacket_r_fr_MSTR}
    rotation=<rot 7.287 10.9701 2.483>
    translation=<-0.909 3.976 10>
   }
   Save cached pose 'Cache_PreSkyDive'(Pose=Component To Local(ComponentPose=@n37)) {cacheName=Cache_PreSkyDive}
  ]}
 EventGraph {nodes [
   Event BlueprintUpdateAnimation()
   > n1 = Gravity Override State Function DEPRECATED(CharacterPartAnimInstance=Self-Reference(), GameTimeInSeconds=GetGameTimeInSeconds(), GravityOverrideParams=Get Gravity Override Params())
   > Set Gravity Override(Gravity Override=@n1)
  ]}
]
variables [
 Gravity Override {type=Vector}
 Gravity Override Params {type=GravityOverrideParamsStruct}
]
