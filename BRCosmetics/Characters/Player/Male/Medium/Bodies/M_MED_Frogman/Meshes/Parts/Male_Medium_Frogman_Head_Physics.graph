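# Physics asset for the head part of the male medium Frogman player cosmetic, simulating a short chain of dynamic face bones.
# A small parent capsule on dyn_parent_face anchors two larger capsules on dyn_face_01 and dyn_face_02.
# Constraints link each body to the previous one with limited swing and tight twist, letting the piece bend while barely rotating along its axis.
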PhysicsAsset /BRCosmetics/Characters/Player/Male/Medium/Bodies/M_MED_Frogman/Meshes/Parts/Male_Medium_Frogman_Head_Physics
bodies {
 dyn_face_01 [{shapeType=Capsule, length=6.5, radius=4, center=<3 0 0>, rotation=<rot 90 0 0>}]
 dyn_face_02 [{shapeType=Capsule, length=4.75, radius=3.5, center=<5 0 0>, rotation=<rot 90 0 0>}]
 dyn_parent_face [{shapeType=Capsule, length=0.505, radius=0.505, rotation=<rot 0 -90 0>}]
}
constraints [
 {bone1Name=dyn_face_01, bone2Name=dyn_parent_face, swing1LimitDegrees=40, swing2LimitDegrees=55, twistLimitDegrees=5}
 {bone1Name=dyn_face_02, bone2Name=dyn_face_01, swing1LimitDegrees=35, swing2LimitDegrees=65, twistLimitDegrees=8}
]
